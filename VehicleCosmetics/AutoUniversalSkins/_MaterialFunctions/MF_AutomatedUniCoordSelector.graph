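# Material function for vehicle cosmetic skins that places a selected trim-sheet cell, such as a logo, on the second UV channel.
# Cell count, selector, transform and rotation inputs build the UVs through CustomRotator, and floored coordinates mask out everything outside the tile.
# An optional mirrored instance adds its own cell and offsets, with a static switch for flipped UVs.
# It outputs UVs, a mask and the masked texture sample.
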MaterialFunction /VehicleCosmetics/AutoUniversalSkins/_MaterialFunctions/MF_AutomatedUniCoordSelector
graph {nodes [
  Constant2Vector() {g=1, r=1}
  Constant() {desc=RadianRemover, r=57.2958}
  Constant() {desc=RadianRemover, r=57.2958}
  n1 = NamedRerouteUsage()
  n2 = Subtract(A=NamedRerouteUsage(), B=NamedRerouteUsage() {desc=TileAmount})
  n3 = Multiply(A=NamedRerouteUsage(), B=NamedRerouteUsage())
  n4 = ComponentMask(None=NamedRerouteUsage()) {g=true, r=true}
  n5 = NamedRerouteUsage()
  n6 = Multiply(A=@n4, B=@n5) {constB=0.5}
  n7 = Subtract(A=@n6, B=@n5)
  n8 = Multiply(A=@n6, B=ComponentMask(None=NamedRerouteUsage()) {a=true, b=true})
  n9 = MaterialFunctionCall(
   UVs=Subtract(A=Multiply(A=@n3, B=@n4), B=Add(A=@n7, B=@n8))
   Rotation Center=Multiply(B=NamedRerouteUsage()) {constA=0.5}
   "Rotation Angle (0-1)"=Divide(A=NamedRerouteUsage(), B=Constant() {r=360})
  ) {materialFunction=&CustomRotator}
  n10 = LinearInterpolate(A=NamedRerouteUsage(), B=Add(A=Multiply(A=@n1, B=@n2), B=@n9.Rotated Values), Alpha=NamedRerouteUsage()) {desc=B_Branch}
  FunctionOutput(None=@n10) {outputName=UVs, sortPriority=20}
  n11 = Floor(None=NamedRerouteUsage())
  n12 = ComponentMask(None=@n11) {r=true}
  n13 = ComponentMask(None=@n11) {g=true}
  n14 = Floor(None=Subtract(A=NamedRerouteUsage(), B=NamedRerouteUsage()))
  n15 = ComponentMask(None=@n14) {r=true}
  n16 = ComponentMask(None=@n14) {g=true}
  n17 = Max(A=Add(A=Abs(None=@n12), B=Abs(None=@n13)), B=Add(A=Abs(None=@n15), B=Abs(None=@n16)))
  n18 = LinearInterpolate(A=NamedRerouteUsage(), B=OneMinus(None=Saturate(None=@n17)), Alpha=NamedRerouteUsage()) {desc=A_Branch}
  FunctionOutput(None=@n18) {outputName=Mask, sortPriority=10}
  n19 = NamedRerouteUsage()
  n20 = TextureSample(
   UVs=NamedRerouteUsage()
   Tex=FunctionInput(Preview=TextureObject() {texture=/VehicleCosmetics/Textures/UniversalTextures/T_AutomatedUniCoordSelector_logoSpriteGuide1}) {bUsePreviewValueAsDefault=true, inputName=TrimSheet, inputType=FunctionInput_Texture2D, sortPriority=-50}
  ) {texture=&T_AutomatedUniCoordSelector_logoSpriteGuide1}
  FunctionOutput(None=Multiply(A=Multiply(A=@n19, B=@n20.RGBA), B=@n20.A)) {description=Full Masked Texture Output, outputName=TextureOutput}
  n21 = FunctionInput(Preview=Constant2Vector() {g=1, r=1}) {
   bUsePreviewValueAsDefault=true
   description="Which tile to select by row, column x,y starting in top left"
   inputName=TrimSheet_CellSelector
   inputType=FunctionInput_Vector2
   sortPriority=15
  }
  NamedRerouteDeclaration(None=@n21) {name=TileSelector, nodeColor=<rgba 0 0.7422 1 1>}
  n22 = FunctionInput(Preview=Constant2Vector() {g=2, r=2}) {
   bUsePreviewValueAsDefault=true
   description="Amount per row, column in x,y"
   inputName=TrimSheet_Cells
   inputType=FunctionInput_Vector2
   sortPriority=10
  }
  n23 = Saturate(None=Ceil(None=Subtract(A=@n22)))
  n24 = LinearInterpolate(A=Constant2Vector() {g=1, r=1}, B=@n22, Alpha=@n23) {constA=1}
  NamedRerouteDeclaration(None=Divide(A=NamedRerouteUsage(), B=@n24)) {name=TileAmount, nodeColor=<rgba 1 0.8203 0 1>}
  n25 = FunctionInput(Preview=Constant()) {
   bUsePreviewValueAsDefault=true
   description="Secondary logo for bodies needing two \"mirrored\" instances"
   inputName=MirroredTrimSheet_Opacity
   inputType=FunctionInput_Scalar
   sortPriority=20
  }
  NamedRerouteDeclaration(None=Saturate(None=Floor(None=@n25))) {name=MirroredTrimSheetOpacity, nodeColor=<rgba 0.1484 0 1 1>}
  NamedRerouteDeclaration(None=@n9.Rotated Values) {name=Shared_Coordinates, nodeColor=<rgba 0 1 0.6016 1>}
  n26 = NamedRerouteUsage()
  n27 = Subtract(A=NamedRerouteUsage(), B=NamedRerouteUsage() {desc=TileAmount})
  n28 = Multiply(A=NamedRerouteUsage(), B=NamedRerouteUsage())
  n29 = ComponentMask(None=NamedRerouteUsage()) {g=true, r=true}
  n30 = NamedRerouteUsage()
  n31 = Multiply(A=@n29, B=@n30) {constB=0.5}
  n32 = Subtract(A=@n31, B=@n30)
  n33 = Multiply(A=@n31, B=ComponentMask(None=NamedRerouteUsage()) {a=true, b=true})
  n34 = MaterialFunctionCall(
   UVs=Subtract(A=Multiply(A=@n28, B=@n29), B=Add(A=@n32, B=@n33))
   Rotation Center=Multiply(B=NamedRerouteUsage()) {constA=0.5}
   "Rotation Angle (0-1)"=Divide(A=NamedRerouteUsage(), B=Constant() {r=360})
  ) {materialFunction=&CustomRotator}
  NamedRerouteDeclaration(None=Add(A=Multiply(A=@n26, B=@n27), B=@n34.Rotated Values)) {name=MirroredUVOutput, nodeColor=<rgba 0 0.5078 1 1>}
  NamedRerouteDeclaration(None=@n34.Rotated Values) {name=Shared_CloneCoordinates, nodeColor=<rgba 0 1 0.6016 1>}
  NamedRerouteDeclaration(None=Constant2Vector() {g=1, r=1}) {name=Shared_Vector2, nodeColor=<rgba 0.6172 0 1 1>}
  n35 = ComponentMask(None=Add(A=TextureCoordinate() {coordinateIndex=1}) {constB=0.5}) {r=true}
  NamedRerouteDeclaration(None=LinearInterpolate(B=Floor(None=@n35), Alpha=NamedRerouteUsage()) {constA=1}) {name=UVMirror_Mask, nodeColor=<rgba 1 0.0703 0 1>}
  n36 = FunctionInput(Preview=StaticBool()) {
   bUsePreviewValueAsDefault=true
   description=Flips the UVs of the mirrored TrimSheet to appear symmetrical
   inputName=MirrorUsesFlippedUVs
   inputType=FunctionInput_StaticBool
   sortPriority=100
  }
  NamedRerouteDeclaration(None=@n36) {name=bUseFlippedUVs, nodeColor=<rgba 0.6641 1 0 1>}
  n37 = TextureCoordinate() {coordinateIndex=1}
  NamedRerouteDeclaration(None=StaticSwitch(
    True=TextureCoordinate() {coordinateIndex=1, desc=-X, uTiling=-1}
    False=NamedRerouteDeclaration(None=@n37) {name=UV2, nodeColor=<rgba 0 1 0.1094 1>}
    Value=NamedRerouteUsage()
   )) {name=FlippedUV2, nodeColor=<rgba 1 0 0.9844 1>}
  n38 = Floor(None=NamedRerouteUsage())
  n39 = Abs(None=ComponentMask(None=@n38) {r=true})
  n40 = Abs(None=ComponentMask(None=@n38) {g=true})
  n41 = Floor(None=Subtract(A=NamedRerouteUsage(), B=NamedRerouteUsage()))
  n42 = Abs(None=ComponentMask(None=@n41) {r=true})
  n43 = Abs(None=ComponentMask(None=@n41) {g=true})
  n44 = Saturate(None=Max(A=Add(A=@n39, B=@n40), B=Add(A=@n42, B=@n43)))
  NamedRerouteDeclaration(None=Multiply(A=NamedRerouteUsage(), B=OneMinus(None=@n44))) {name=MirroredTrimSheetMask, nodeColor=<rgba 0.8281 1 0 1>}
  NamedRerouteDeclaration(None=@n10) {name=Result_UVs, nodeColor=<rgba 0.9219 1 0 1>}
  NamedRerouteDeclaration(None=@n23) {name=bIsTileSheet, nodeColor=<rgba 0.1953 1 0 1>}
  NamedRerouteDeclaration(None=@n18) {name=Result_Mask, nodeColor=<rgba 1 0.3516 0 1>}
  NamedRerouteDeclaration(None=@n24) {name=TrimSheet_Cells, nodeColor=<rgba 0.5234 1 0 1>}
  n45 = FunctionInput(Preview=Constant()) {
   bUsePreviewValueAsDefault=true
   description=Degrees counterclockwise
   inputName=TrimSheet_Rotation
   inputType=FunctionInput_Scalar
   sortPriority=5
  }
  NamedRerouteDeclaration(None=@n45) {name=TrimSheetRotation, nodeColor=<rgba 0 0.7422 1 1>}
  n46 = MaterialFunctionCall(Float2=NamedRerouteUsage()) {materialFunction=&BreakOutFloat2Components}
  n47 = Divide(A=@n46.R, B=@n46.R)
  n48 = Divide(A=@n46.G, B=@n46.G)
  n49 = Saturate(None=Floor(None=@n48))
  n50 = Saturate(None=Floor(None=@n47))
  n51 = Multiply(A=AppendVector(A=LinearInterpolate(A=@n47, Alpha=@n49), B=LinearInterpolate(A=@n48, Alpha=@n50)), B=Constant() {r=12})
  n52 = FunctionInput(Preview=Constant4Vector() {constant=<rgba 1 1 0 0>}.RGBA) {bUsePreviewValueAsDefault=true, description="RG=XYScale BA=XYOffset", inputName=TrimSheet_Transform, inputType=FunctionInput_Vector4}
  n53 = ComponentMask(None=@n52) {g=true, r=true}
  NamedRerouteDeclaration(None=AppendVector(A=Divide(A=@n51, B=@n53), B=ComponentMask(None=@n52) {a=true, b=true})) {name=TrimSheetTransform, nodeColor=<rgba 0 1 1 1>}
  NamedRerouteDeclaration(None=Divide(B=@n24) {constA=0.5}) {name=TileOffset, nodeColor=<rgba 0.1953 1 0 1>}
  n54 = FunctionInput(Preview=NamedRerouteUsage()) {
   bUsePreviewValueAsDefault=true
   description="Which tile to select by row, column x,y, starting in top left (Defaults to TrimSheet_CellSelector)"
   inputName=MirroredTrimSheet_CellSelector
   inputType=FunctionInput_Vector2
   sortPriority=35
  }
  NamedRerouteDeclaration(None=@n54) {name=MirroredTileSelector, nodeColor=<rgba 0 0.0391 1 1>}
  n55 = FunctionInput(Preview=Constant()) {
   bUsePreviewValueAsDefault=true
   description=Degrees counterclockwise
   inputName=MirroredTrimSheet_RotationOffset
   inputType=FunctionInput_Scalar
   sortPriority=30
  }
  n56 = Add(A=NamedRerouteUsage(), B=@n55)
  NamedRerouteDeclaration(None=StaticSwitch(True=@n56, False=Multiply(A=@n56) {constB=-1}, Value=NamedRerouteUsage())) {name=MirroredTrimSheetRotation, nodeColor=<rgba 0.4531 0 1 1>}
  n57 = FunctionInput(Preview=Constant4Vector().RGBA) {
   bUsePreviewValueAsDefault=true
   description="RG=XYScale BA=XYOffset"
   inputName=MirroredTrimSheet_TransformOffset
   inputType=FunctionInput_Vector4
   sortPriority=25
  }
  n58 = ComponentMask(None=@n57) {g=true, r=true}
  n59 = Add(A=NamedRerouteUsage(), B=AppendVector(A=Multiply(A=@n58) {constB=-1}, B=ComponentMask(None=@n57) {a=true, b=true}))
  n60 = ComponentMask(None=@n59) {b=true}
  n61 = ComponentMask(None=@n59) {a=true}
  n62 = AppendVector(A=Add(A=@n60) {constB=-2}, B=@n61)
  n63 = ComponentMask(None=@n59) {g=true, r=true}
  n64 = AppendVector(A=Multiply(A=@n60) {constB=-1}, B=@n61)
  NamedRerouteDeclaration(None=StaticSwitch(True=AppendVector(A=@n63, B=@n62), False=AppendVector(A=@n63, B=@n64), Value=NamedRerouteUsage())) {name=MirroredTrimSheetTransform, nodeColor=<rgba 1 0 0.7266 1>}
 ]}
