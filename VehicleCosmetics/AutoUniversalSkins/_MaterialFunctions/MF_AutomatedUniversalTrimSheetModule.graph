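# Material function that exposes a trim-sheet logo layer for automated universal vehicle skins.
# A texture object and UniversalLogos parameters for transform, rotation and sprite grid cell feed MF_AutomatedUniCoordSelector.
# Further parameters control an optional mirrored copy with its own opacity, transform offset and rotation offset.
# The selector's texture output splits into RGB base color and a pre-calculated alpha mask.
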MaterialFunction /VehicleCosmetics/AutoUniversalSkins/_MaterialFunctions/MF_AutomatedUniversalTrimSheetModule
graph {nodes [
  n1 = TextureObjectParameter() {
   desc=RGB +Alpha opacity
   group=UniversalLogos
   parameterName=TrimSheet
   texture=/VehicleCosmetics/Textures/UniversalTextures/T_AutomatedUniCoordSelector_logoSpriteGuide1
  }
  n2 = VectorParameter() {
   desc="RG=XYScale BA=XYOffset"
   group=UniversalLogos
   parameterName=TrimSheet1_Transform
   channelNames {a=Y Offset, b=X Offset, g=Y Scale, r=X Scale}
   defaultValue=<rgba 2 1.25 0 0.69>
  }
  n3 = VectorParameter() {
   desc="Number of columns and rows for X,Y=r,g  Select cell starting position starting from top left in X,Y = b,a. Row and Column both start with index 1/1."
   group=UniversalLogos
   parameterName=TrimSheet1_Cells
   channelNames {a=Sprite Y Position, b=Sprite X Position, g="##Sprites in Y", r="##Sprites in X"}
   defaultValue=<rgba 1 1 1 1>
  }
  n4 = VectorParameter() {
   desc="RG=XY Scale Offset; BA=XY Position Offset"
   group=UniversalLogos
   parameterName=TrimSheet1Mirrored_Offset
   channelNames {a=Y Offset, b=X Offset, g=Y Scale, r=X Scale}
  }
  n5 = MaterialFunctionCall!(
   TrimSheet=@n1
   TrimSheet_Transform=AppendVector(A=@n2, B=@n2)
   TrimSheet_Rotation=ScalarParameter() {desc=Degrees counterclockwise, group=UniversalLogos, parameterName=TrimSheet1_Rotation}
   TrimSheet_Cells=AppendVector(A=@n3."##Sprites in X", B=@n3."##Sprites in X")
   TrimSheet_CellSelector=AppendVector(A=@n3.Sprite X Position, B=@n3.Sprite X Position)
   MirroredTrimSheet_Opacity=ScalarParameter() {desc="Enables the mirrored TrimSheet; 0 = hidden; 1 = visible", group=UniversalLogos, parameterName=TrimSheet1Mirrored_Opacity}
   MirroredTrimSheet_TransformOffset=AppendVector(A=@n4, B=@n4)
   MirroredTrimSheet_RotationOffset=ScalarParameter() {desc=Offset by degrees counterclockwise, group=UniversalLogos, parameterName=TrimSheet1Mirrored_RotationOffset}
   MirroredTrimSheet_CellSelector
   MirrorUsesFlippedUVs
  ) {materialFunction=&/MF_AutomatedUniCoordSelector}
  FunctionOutput(None=ComponentMask(None=@n5.TextureOutput) {b=true, g=true, r=true}) {outputName=TrimSheet_BaseColor}
  FunctionOutput(None=ComponentMask(None=@n5.TextureOutput) {a=true, desc=Pre-Calculated Alpha/Mask}) {outputName=TrimSheet_Mask, sortPriority=1}
 ]}
