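# Base material function for automated universal vehicle skins, building team, accent and paint colored bodywork as material attributes.
# Skin mask channels blend colors, metallic, roughness, detail textures and clear coat thickness, roughness and bottom normals.
# A trim sheet module supplies automated mask and base color, processed alongside body textures through composite subgraphs.
# Static switches select variants, including modified legacy Rocket League sharp-mask features with a clear-coat-specific mask.
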MaterialFunction /VehicleCosmetics/AutoUniversalSkins/_MaterialFunctions/MF_Vehicle_Skin_AutomatedUniversal_Base
graph {expressions {
  MaterialExpressionAbs_0 {class=MaterialExpressionAbs, inputs [None {from=MaterialExpressionAdd_0, output=""}], properties {}}
  MaterialExpressionAdd_0 {
   class=MaterialExpressionAdd
   inputs [
    A {from=MaterialExpressionConstant_0}
    B {from=MaterialExpressionPower_1}
   ]
  }
  MaterialExpressionAdd_1 {
   class=MaterialExpressionAdd
   inputs [
    A {from=MaterialExpressionComponentMask_1}
    B {from=MaterialExpressionComponentMask_1}
   ]
  }
  MaterialExpressionAdd_2 {
   class=MaterialExpressionAdd
   inputs [
    A {from=MaterialExpressionClamp_0}
    B {from=""}
   ]
   properties {constB=-1}
  }
  MaterialExpressionAdd_3 {
   class=MaterialExpressionAdd
   inputs [
    A {from=MaterialExpressionClamp_2}
    B {from=""}
   ]
   properties {constB=-1}
  }
  MaterialExpressionAdd_4 {
   class=MaterialExpressionAdd
   inputs [
    A {from=MaterialExpressionAdd_1}
    B {from=MaterialExpressionPower_5}
   ]
  }
  MaterialExpressionAppendVector_0 {
   class=MaterialExpressionAppendVector
   inputs [
    A {from=MaterialExpressionTextureSampleParameter2D_0, output=B}
    B {from=MaterialExpressionTextureSampleParameter2D_0, output=B}
   ]
  }
  MaterialExpressionAppendVector_2 {
   class=MaterialExpressionAppendVector
   inputs [
    A {from=MaterialExpressionReroute_42, output=RGB}
    B {from=MaterialExpressionNamedRerouteUsage_45, output=""}
   ]
  }
  MaterialExpressionAppendVector_3 {
   class=MaterialExpressionAppendVector
   inputs [
    A {from=MaterialExpressionReroute_42, output=RGB}
    B {from=MaterialExpressionConstant_1, output=""}
   ]
  }
  MaterialExpressionClamp_0 {
   class=MaterialExpressionClamp
   inputs [
    None {from=MaterialExpressionScalarParameter_16, output=""}
    Min {from="", output=""}
    Max {from="", output=""}
   ]
   properties {maxDefault=2}
  }
  MaterialExpressionClamp_1 {
   class=MaterialExpressionClamp
   inputs [
    None {from=MaterialExpressionPower_7, output=""}
    Min {from=MaterialExpressionConstant_6, output=""}
    Max {from="", output=""}
   ]
   properties {clampMode=CMODE_ClampMin, minDefault=0.5}
  }
  MaterialExpressionClamp_2 {
   class=MaterialExpressionClamp
   inputs [
    None {from=MaterialExpressionScalarParameter_17, output=""}
    Min {from="", output=""}
    Max {from="", output=""}
   ]
   properties {maxDefault=2}
  }
  MaterialExpressionComponentMask_0 {class=MaterialExpressionComponentMask, inputs [None {from=MaterialExpressionReroute_30, output=SkinMaskTexture}], properties {r=true}}
  MaterialExpressionComponentMask_1 {class=MaterialExpressionComponentMask, inputs [None {from=MaterialExpressionAbs_0}], properties {g=true}}
  MaterialExpressionComponentMask_2 {class=MaterialExpressionComponentMask, inputs [None {from=MaterialExpressionNamedRerouteUsage_34}], properties {r=true}}
  MaterialExpressionComponentMask_3 {class=MaterialExpressionComponentMask, inputs [None {from=MaterialExpressionNamedRerouteUsage_35}], properties {g=true}}
  MaterialExpressionComponentMask_4 {class=MaterialExpressionComponentMask, inputs [None {from=MaterialExpressionReroute_56, output=RGBA}], properties {a=true}}
  MaterialExpressionComponentMask_5 {class=MaterialExpressionComponentMask, inputs [None {from=MaterialExpressionReroute_56, output=RGBA}], properties {b=true, g=true, r=true}}
  MaterialExpressionComposite_0 {
   class=MaterialExpressionComposite
   inputs [SkinMaskTexture {from=MaterialExpressionNamedRerouteUsage_12, output=""}]
   properties {subgraphName=BodyFinish}
  }
  MaterialExpressionComposite_1 {
   class=MaterialExpressionComposite
   inputs [
    UV {from=MaterialExpressionNamedRerouteUsage_25, output=""}
    UV Detail {from=MaterialExpressionNamedRerouteUsage_26, output=""}
    UV BottomNormal {from=MaterialExpressionNamedRerouteUsage_21, output=""}
    LOD {from=MaterialExpressionScalarParameter_1, output=""}
    BottomNormalFlatness {from=MaterialExpressionScalarParameter_14, output=""}
   ]
   properties {subgraphName=BodyTextures}
  }
  MaterialExpressionComposite_10 {
   class=MaterialExpressionComposite
   inputs [
    BaseColorInput {from=MaterialExpressionReroute_55, output=Output}
    TextureInput {from=MaterialExpressionNamedRerouteUsage_27, output=""}
   ]
   properties {desc=Legacy - Rocket League Feature, subgraphName=SharpMasksBaseColorMultiply}
  }
  MaterialExpressionComposite_11 {
   class=MaterialExpressionComposite
   inputs [RGBA {from=MaterialExpressionReroute_45, output=SharpMasksSwitchOutput}]
   properties {subgraphName=ProcessAutoUniMasks}
  }
  MaterialExpressionComposite_2 {
   class=MaterialExpressionComposite
   inputs [
    DetailBaseTexture {from=MaterialExpressionNamedRerouteUsage_29, output=""}
    DetailAccentTexture {from=MaterialExpressionNamedRerouteUsage_30, output=""}
    SkinMaskTexture {from=MaterialExpressionNamedRerouteUsage_18, output=""}
   ]
   properties {subgraphName=Details}
  }
  MaterialExpressionComposite_3 {
   class=MaterialExpressionComposite
   inputs [
    Roughness {from=MaterialExpressionNamedRerouteUsage_5, output=""}
    SkinMask {from=MaterialExpressionNamedRerouteUsage_4, output=""}
   ]
   properties {subgraphName=Roughness}
  }
  MaterialExpressionComposite_4 {
   class=MaterialExpressionComposite
   inputs [
    TeamColor {from=MaterialExpressionNamedRerouteUsage_8, output=""}
    AccentColor {from=MaterialExpressionNamedRerouteUsage_9, output=""}
    PaintColor {from=MaterialExpressionNamedRerouteUsage_10, output=""}
    SkinMaskTexture {from=MaterialExpressionNamedRerouteUsage_7, output=""}
   ]
   properties {subgraphName=BaseColor}
  }
  MaterialExpressionComposite_5 {
   class=MaterialExpressionComposite
   inputs [
    BaseColor {from=MaterialExpressionReroute_26, output=BaseColor}
    BaseColorTexture {from=MaterialExpressionNamedRerouteUsage_15, output=""}
    SkinMaskTexture {from=MaterialExpressionNamedRerouteUsage_14, output=""}
   ]
   properties {subgraphName=UseBaseColorTexture}
  }
  MaterialExpressionComposite_6 {class=MaterialExpressionComposite, inputs [], properties {subgraphName=Vectors}}
  MaterialExpressionComposite_7 {
   class=MaterialExpressionComposite
   inputs [
    SkinMaskTexture {from=MaterialExpressionNamedRerouteUsage_43, output=""}
    BottomNormalBaseTexture {from=MaterialExpressionNamedRerouteUsage_11, output=""}
    BottomNormalAccentTexture {from=MaterialExpressionNamedRerouteUsage_19, output=""}
   ]
   properties {subgraphName=ClearCoat}
  }
  MaterialExpressionComposite_8 {
   class=MaterialExpressionComposite
   inputs [RGB {from=MaterialExpressionReroute_44, output=SkinMask}]
   properties {desc=Legacy - Rocket League Feature | Modified to support Clear Coat specific Mask, subgraphName=ProcessSharpMasksMODIFIED}
  }
  MaterialExpressionComposite_9 {
   class=MaterialExpressionComposite
   inputs [BaseColor {from=MaterialExpressionReroute_28, output=BaseColor}]
   properties {subgraphName=UseAutoUniBaseColor}
  }
  MaterialExpressionConstant3Vector_1 {class=MaterialExpressionConstant3Vector, properties {constant=<rgba 0 0 1 0>}}
  MaterialExpressionConstant3Vector_4 {class=MaterialExpressionConstant3Vector, properties {constant=<rgba 0 0 1 0>}}
  MaterialExpressionConstant3Vector_5 {class=MaterialExpressionConstant3Vector, properties {constant=<rgba 0 1 0 0>}}
  MaterialExpressionConstant3Vector_6 {class=MaterialExpressionConstant3Vector}
  MaterialExpressionConstantBiasScale_0 {class=MaterialExpressionConstantBiasScale, inputs [None {from=MaterialExpressionPower_0}], properties {bias=-0.5, scale=4}}
  MaterialExpressionConstant_0 {class=MaterialExpressionConstant, properties {r=-0.5}}
  MaterialExpressionConstant_1 {class=MaterialExpressionConstant}
  MaterialExpressionConstant_6 {class=MaterialExpressionConstant, properties {r=0.5}}
  MaterialExpressionFloor_0 {class=MaterialExpressionFloor, inputs [None {from=MaterialExpressionSaturate_3, output=""}], properties {}}
  MaterialExpressionFloor_1 {class=MaterialExpressionFloor, inputs [None {from=MaterialExpressionSaturate_4, output=""}], properties {}}
  MaterialExpressionFloor_2 {class=MaterialExpressionFloor, inputs [None {from=MaterialExpressionSaturate_5, output=""}], properties {}}
  MaterialExpressionFloor_3 {class=MaterialExpressionFloor, inputs [None {from=MaterialExpressionSaturate_6, output=""}], properties {}}
  MaterialExpressionFunctionOutput_0 {class=MaterialExpressionFunctionOutput, inputs [None {from=MaterialExpressionSetMaterialAttributes_0, output=""}]}
  MaterialExpressionLinearInterpolate_0 {
   class=MaterialExpressionLinearInterpolate
   inputs [
    A {from=MaterialExpressionScalarParameter_2}
    B {from=MaterialExpressionScalarParameter_3}
    Alpha {from=MaterialExpressionMaterialFunctionCall_0, output=G}
   ]
  }
  MaterialExpressionLinearInterpolate_1 {
   class=MaterialExpressionLinearInterpolate
   inputs [
    A {from=MaterialExpressionLinearInterpolate_0}
    B {from=MaterialExpressionScalarParameter_4}
    Alpha {from=MaterialExpressionMaterialFunctionCall_0, output=B}
   ]
  }
  MaterialExpressionLinearInterpolate_10 {
   class=MaterialExpressionLinearInterpolate
   inputs [
    A {from=MaterialExpressionReroute_29, output=BaseColorTexture}
    B {from=MaterialExpressionReroute_27, output=BaseColor}
    Alpha {from=MaterialExpressionComponentMask_0}
   ]
   properties {desc=Put diffuse where R skin mask is black}
  }
  MaterialExpressionLinearInterpolate_11 {
   class=MaterialExpressionLinearInterpolate
   inputs [
    A {from=MaterialExpressionScalarParameter_8}
    B {from=MaterialExpressionScalarParameter_9}
    Alpha {from=MaterialExpressionMaterialFunctionCall_8, output=G}
   ]
  }
  MaterialExpressionLinearInterpolate_12 {
   class=MaterialExpressionLinearInterpolate
   inputs [
    A {from=MaterialExpressionLinearInterpolate_11}
    B {from=MaterialExpressionScalarParameter_10}
    Alpha {from=MaterialExpressionMaterialFunctionCall_8, output=B}
   ]
  }
  MaterialExpressionLinearInterpolate_13 {
   class=MaterialExpressionLinearInterpolate
   inputs [
    A {from=MaterialExpressionReroute_35, output=BottomNormalBaseTexture}
    B {from=MaterialExpressionReroute_36, output=BottomNormalAccentTexture}
    Alpha {from=MaterialExpressionMaterialFunctionCall_3, output=G}
   ]
  }
  MaterialExpressionLinearInterpolate_14 {
   class=MaterialExpressionLinearInterpolate
   inputs [
    A {from=MaterialExpressionLinearInterpolate_13}
    B {from=MaterialExpressionReroute_36, output=BottomNormalAccentTexture}
    Alpha {from=MaterialExpressionMaterialFunctionCall_3, output=B}
   ]
  }
  MaterialExpressionLinearInterpolate_15 {
   class=MaterialExpressionLinearInterpolate
   inputs [
    A {from=MaterialExpressionReroute_31, output=DetailBaseTexture}
    B {from=MaterialExpressionReroute_34, output=DetailAccentTexture}
    Alpha {from=MaterialExpressionSaturate_1}
   ]
   properties {constA=1}
  }
  MaterialExpressionLinearInterpolate_16 {
   class=MaterialExpressionLinearInterpolate
   inputs [
    A {from=MaterialExpressionNamedRerouteUsage_33}
    B {from=MaterialExpressionLinearInterpolate_19}
    Alpha {from=MaterialExpressionFloor_0}
   ]
  }
  MaterialExpressionLinearInterpolate_17 {
   class=MaterialExpressionLinearInterpolate
   inputs [
    A {from=MaterialExpressionScalarParameter_11}
    B {from=MaterialExpressionScalarParameter_12}
    Alpha {from=MaterialExpressionMaterialFunctionCall_19, output=G}
   ]
  }
  MaterialExpressionLinearInterpolate_18 {
   class=MaterialExpressionLinearInterpolate
   inputs [
    A {from=MaterialExpressionLinearInterpolate_17}
    B {from=MaterialExpressionScalarParameter_13}
    Alpha {from=MaterialExpressionMaterialFunctionCall_19, output=B}
   ]
  }
  MaterialExpressionLinearInterpolate_19 {
   class=MaterialExpressionLinearInterpolate
   inputs [
    A {from=MaterialExpressionNamedRerouteUsage_38}
    B {from=MaterialExpressionNamedRerouteUsage_39}
    Alpha {from=MaterialExpressionFloor_1}
   ]
  }
  MaterialExpressionLinearInterpolate_2 {
   class=MaterialExpressionLinearInterpolate
   inputs [
    A {from=MaterialExpressionScalarParameter_5}
    B {from=MaterialExpressionScalarParameter_6}
    Alpha {from=MaterialExpressionMaterialFunctionCall_2, output=G}
   ]
  }
  MaterialExpressionLinearInterpolate_20 {
   class=MaterialExpressionLinearInterpolate
   inputs [
    A {from=MaterialExpressionConstant3Vector_6}
    B {from=MaterialExpressionLinearInterpolate_21}
    Alpha {from=MaterialExpressionFloor_3}
   ]
  }
  MaterialExpressionLinearInterpolate_21 {
   class=MaterialExpressionLinearInterpolate
   inputs [
    A {from=MaterialExpressionConstant3Vector_5}
    B {from=MaterialExpressionConstant3Vector_4}
    Alpha {from=MaterialExpressionFloor_2}
   ]
  }
  MaterialExpressionLinearInterpolate_24 {
   class=MaterialExpressionLinearInterpolate
   inputs [
    A {from=MaterialExpressionNamedRerouteUsage_42}
    B {from=MaterialExpressionSaturate_10}
    Alpha {from=MaterialExpressionComponentMask_4}
   ]
  }
  MaterialExpressionLinearInterpolate_3 {
   class=MaterialExpressionLinearInterpolate
   inputs [
    A {from=MaterialExpressionLinearInterpolate_2}
    B {from=MaterialExpressionScalarParameter_7}
    Alpha {from=MaterialExpressionMaterialFunctionCall_2, output=B}
   ]
  }
  MaterialExpressionLinearInterpolate_4 {
   class=MaterialExpressionLinearInterpolate
   inputs [
    A {from=MaterialExpressionConstant3Vector_1}
    B {from=MaterialExpressionTextureSampleParameter2D_5, output=RGB}
    Alpha {from=MaterialExpressionReroute_21, output=LOD}
   ]
   properties {constAlpha=0}
  }
  MaterialExpressionLinearInterpolate_5 {
   class=MaterialExpressionLinearInterpolate
   inputs [
    A {from=MaterialExpressionMaterialFunctionCall_5, output=R}
    B {from=MaterialExpressionMaterialFunctionCall_5, output=R}
    Alpha {from=MaterialExpressionMaterialFunctionCall_4, output=G}
   ]
  }
  MaterialExpressionLinearInterpolate_6 {
   class=MaterialExpressionLinearInterpolate
   inputs [
    A {from=MaterialExpressionLinearInterpolate_5}
    B {from=MaterialExpressionMaterialFunctionCall_5, output=B}
    Alpha {from=MaterialExpressionMaterialFunctionCall_4, output=B}
   ]
  }
  MaterialExpressionLinearInterpolate_7 {
   class=MaterialExpressionLinearInterpolate
   inputs [
    A {from=MaterialExpressionReroute_8, output=TeamColor}
    B {from=MaterialExpressionReroute_13, output=AccentColor}
    Alpha {from=MaterialExpressionMaterialFunctionCall_6, output=G}
   ]
  }
  MaterialExpressionLinearInterpolate_8 {
   class=MaterialExpressionLinearInterpolate
   inputs [
    A {from=MaterialExpressionLinearInterpolate_7}
    B {from=MaterialExpressionReroute_18, output=PaintColor}
    Alpha {from=MaterialExpressionMaterialFunctionCall_6, output=B}
   ]
  }
  MaterialExpressionLinearInterpolate_9 {
   class=MaterialExpressionLinearInterpolate
   inputs [
    A {from=MaterialExpressionReroute_58}
    B {from=MaterialExpressionStaticSwitchParameter_8}
    Alpha {from=MaterialExpressionNamedRerouteUsage_0}
   ]
  }
  MaterialExpressionMaterialFunctionCall_0 {
   class=MaterialExpressionMaterialFunctionCall
   inputs [Color {from=MaterialExpressionReroute_10, output=SkinMaskTexture}]
   properties {materialFunction=&SplitComponents}
  }
  MaterialExpressionMaterialFunctionCall_1 {
   class=MaterialExpressionMaterialFunctionCall
   inputs [
    Normal {from=MaterialExpressionTextureSampleParameter2D_6, output=RGB}
    Flatness {from=MaterialExpressionReroute_40}
   ]
   properties {materialFunction=&FlattenNormal}
  }
  MaterialExpressionMaterialFunctionCall_10 {
   class=MaterialExpressionMaterialFunctionCall
   inputs [Color {from=MaterialExpressionReroute_16, output=SkinMaskTexture}]
   properties {materialFunction=&SplitComponents}
  }
  MaterialExpressionMaterialFunctionCall_12 {
   class=MaterialExpressionMaterialFunctionCall
   inputs []
   properties {materialFunction=&/MF_AutomatedUniversalTrimSheetModule}
  }
  MaterialExpressionMaterialFunctionCall_19 {
   class=MaterialExpressionMaterialFunctionCall
   inputs [Color {from=MaterialExpressionReroute_11, output=SkinMaskTexture}]
   properties {materialFunction=&SplitComponents}
  }
  MaterialExpressionMaterialFunctionCall_2 {
   class=MaterialExpressionMaterialFunctionCall
   inputs [Color {from=MaterialExpressionReroute_10, output=SkinMaskTexture}]
   properties {materialFunction=&SplitComponents}
  }
  MaterialExpressionMaterialFunctionCall_3 {
   class=MaterialExpressionMaterialFunctionCall
   inputs [Color {from=MaterialExpressionReroute_11, output=SkinMaskTexture}]
   properties {materialFunction=&SplitComponents}
  }
  MaterialExpressionMaterialFunctionCall_4 {
   class=MaterialExpressionMaterialFunctionCall
   inputs [Color {from=MaterialExpressionReroute_3, output=SkinMask}]
   properties {materialFunction=&SplitComponents}
  }
  MaterialExpressionMaterialFunctionCall_5 {
   class=MaterialExpressionMaterialFunctionCall
   inputs [Color {from=MaterialExpressionReroute_2, output=Roughness}]
   properties {materialFunction=&SplitComponents}
  }
  MaterialExpressionMaterialFunctionCall_6 {
   class=MaterialExpressionMaterialFunctionCall
   inputs [Color {from=MaterialExpressionReroute_20, output=SkinMaskTexture}]
   properties {materialFunction=&SplitComponents}
  }
  MaterialExpressionMaterialFunctionCall_7 {
   class=MaterialExpressionMaterialFunctionCall
   inputs [
    BaseNormal {from=MaterialExpressionNamedRerouteUsage_31}
    AdditionalNormal {from=MaterialExpressionLinearInterpolate_14}
   ]
   properties {materialFunction=&BlendAngleCorrectedNormals}
  }
  MaterialExpressionMaterialFunctionCall_8 {
   class=MaterialExpressionMaterialFunctionCall
   inputs [Color {from=MaterialExpressionReroute_11, output=SkinMaskTexture}]
   properties {materialFunction=&SplitComponents}
  }
  MaterialExpressionMaterialFunctionCall_9 {
   class=MaterialExpressionMaterialFunctionCall
   inputs [
    Normal {from=MaterialExpressionTextureSampleParameter2D_7, output=RGB}
    Flatness {from=MaterialExpressionReroute_40}
   ]
   properties {materialFunction=&FlattenNormal}
  }
  MaterialExpressionMultiply_0 {
   class=MaterialExpressionMultiply
   inputs [
    A {from=MaterialExpressionReroute_49, output=Result}
    B {from=MaterialExpressionNamedRerouteUsage_1}
   ]
  }
  MaterialExpressionMultiply_1 {
   class=MaterialExpressionMultiply
   inputs [
    A {from=MaterialExpressionReroute_47, output=BaseColorInput}
    B {from=MaterialExpressionSaturate_2}
   ]
  }
  MaterialExpressionMultiply_2 {
   class=MaterialExpressionMultiply
   inputs [
    A {from=MaterialExpressionMultiply_0}
    B {from=MaterialExpressionNamedRerouteUsage_24}
   ]
  }
  MaterialExpressionMultiply_3 {
   class=MaterialExpressionMultiply
   inputs [
    A {from=MaterialExpressionTextureCoordinate_0}
    B {from=MaterialExpressionScalarParameter_15}
   ]
  }
  MaterialExpressionMultiply_5 {
   class=MaterialExpressionMultiply
   inputs [
    A {from=MaterialExpressionTextureCoordinate_1}
    B {from=MaterialExpressionScalarParameter_0}
   ]
  }
  MaterialExpressionNamedRerouteDeclaration_0 {
   class=MaterialExpressionNamedRerouteDeclaration
   inputs [None {from=MaterialExpressionNamedRerouteUsage_2}]
   properties {name=MetallicMasked, nodeColor=<rgba 0 1 0.7656 1>}
  }
  MaterialExpressionNamedRerouteDeclaration_1 {
   class=MaterialExpressionNamedRerouteDeclaration
   inputs [None {from=MaterialExpressionStaticSwitchParameter_4}]
   properties {name=UV, nodeColor=<rgba 1 0.8646 0.8646 1>}
  }
  MaterialExpressionNamedRerouteDeclaration_10 {
   class=MaterialExpressionNamedRerouteDeclaration
   inputs [None {from=MaterialExpressionReroute_59, output=SkinMask}]
   properties {name=SkinMaskTexture, nodeColor=<rgba 0.6038 0.3325 0.013 1>}
  }
  MaterialExpressionNamedRerouteDeclaration_11 {
   class=MaterialExpressionNamedRerouteDeclaration
   inputs [None {from=MaterialExpressionReroute_5, output=RoughnessMasked}]
   properties {name=RoughnessMasked, nodeColor=<rgba 0 1 0.7656 1>}
  }
  MaterialExpressionNamedRerouteDeclaration_12 {
   class=MaterialExpressionNamedRerouteDeclaration
   inputs [None {from=MaterialExpressionReroute_23, output=TeamColor}]
   properties {name=TeamColor, nodeColor=<rgba 0.9219 0 1 1>}
  }
  MaterialExpressionNamedRerouteDeclaration_13 {
   class=MaterialExpressionNamedRerouteDeclaration
   inputs [None {from=MaterialExpressionReroute_24, output=AccentColor}]
   properties {name=AccentColor, nodeColor=<rgba 0.9219 0 1 1>}
  }
  MaterialExpressionNamedRerouteDeclaration_14 {
   class=MaterialExpressionNamedRerouteDeclaration
   inputs [None {from=MaterialExpressionReroute_25, output=PaintColor}]
   properties {name=PaintColor, nodeColor=<rgba 0.9219 0 1 1>}
  }
  MaterialExpressionNamedRerouteDeclaration_15 {
   class=MaterialExpressionNamedRerouteDeclaration
   inputs [None {from=MaterialExpressionReroute_15, output=BottomNormalBase}]
   properties {name=BottomNormalBaseTexture, nodeColor=<rgba 0.6038 0.3325 0.013 1>}
  }
  MaterialExpressionNamedRerouteDeclaration_16 {
   class=MaterialExpressionNamedRerouteDeclaration
   inputs [None {from=MaterialExpressionReroute_7, output=BaseColor}]
   properties {name=BaseColorTexture, nodeColor=<rgba 0.6038 0.3325 0.013 1>}
  }
  MaterialExpressionNamedRerouteDeclaration_17 {
   class=MaterialExpressionNamedRerouteDeclaration
   inputs [None {from=MaterialExpressionMultiply_2}]
   properties {name=BaseColor, nodeColor=<rgba 0 1 0.7656 1>}
  }
  MaterialExpressionNamedRerouteDeclaration_18 {
   class=MaterialExpressionNamedRerouteDeclaration
   inputs [None {from=MaterialExpressionNamedRerouteUsage_20}]
   properties {name=CCThicknessMasked, nodeColor=<rgba 0 1 0.7656 1>}
  }
  MaterialExpressionNamedRerouteDeclaration_19 {
   class=MaterialExpressionNamedRerouteDeclaration
   inputs [None {from=MaterialExpressionNamedRerouteUsage_17}]
   properties {name=CCRoughnessMasked, nodeColor=<rgba 0 1 0.7656 1>}
  }
  MaterialExpressionNamedRerouteDeclaration_2 {
   class=MaterialExpressionNamedRerouteDeclaration
   inputs [None {from=MaterialExpressionReroute_6, output=Metallic}]
   properties {name=Metallic, nodeColor=<rgba 0 0.1434 1 1>}
  }
  MaterialExpressionNamedRerouteDeclaration_20 {
   class=MaterialExpressionNamedRerouteDeclaration
   inputs [None {from=MaterialExpressionReroute_38, output=BottomNormalAccent}]
   properties {name=BottomNormalAccentTexture, nodeColor=<rgba 0.6038 0.3325 0.013 1>}
  }
  MaterialExpressionNamedRerouteDeclaration_21 {
   class=MaterialExpressionNamedRerouteDeclaration
   inputs [None {from=MaterialExpressionReroute_19, output=AO}]
   properties {name=AOTexture, nodeColor=<rgba 0.6038 0.3325 0.013 1>}
  }
  MaterialExpressionNamedRerouteDeclaration_22 {
   class=MaterialExpressionNamedRerouteDeclaration
   inputs [None {from=MaterialExpressionReroute_14, output=BodyNormalLOD}]
   properties {name=BodyNormalLOD, nodeColor=<rgba 0.6038 0.3325 0.013 1>}
  }
  MaterialExpressionNamedRerouteDeclaration_23 {
   class=MaterialExpressionNamedRerouteDeclaration
   inputs [None {from=MaterialExpressionReroute_37, output=ClearCoatBottomNormal}]
   properties {name=CCBottomNormal, nodeColor=<rgba 0 0.1434 1 1>}
  }
  MaterialExpressionNamedRerouteDeclaration_24 {
   class=MaterialExpressionNamedRerouteDeclaration
   inputs [None {from=MaterialExpressionReroute_53, output=PBR}]
   properties {name=BodyTextures_RoughnessMetallic, nodeColor=<rgba 1 0.3516 0 1>}
  }
  MaterialExpressionNamedRerouteDeclaration_25 {
   class=MaterialExpressionNamedRerouteDeclaration
   inputs [None {from=MaterialExpressionMultiply_3}]
   properties {name=UV_BottomNormal, nodeColor=<rgba 0.6458 0.6458 0.6458 1>}
  }
  MaterialExpressionNamedRerouteDeclaration_26 {
   class=MaterialExpressionNamedRerouteDeclaration
   inputs [None {from=MaterialExpressionReroute_46, output=TextureRawOutput}]
   properties {name=SkinMaskTextureRaw, nodeColor=<rgba 1 0 0.2812 1>}
  }
  MaterialExpressionNamedRerouteDeclaration_27 {
   class=MaterialExpressionNamedRerouteDeclaration
   inputs [None {from=MaterialExpressionMaterialFunctionCall_12, output=TrimSheet_BaseColor}]
   properties {name=AutoUniBaseColor, nodeColor=<rgba 0.3125 1 0 1>}
  }
  MaterialExpressionNamedRerouteDeclaration_28 {
   class=MaterialExpressionNamedRerouteDeclaration
   inputs [None {from=MaterialExpressionStaticSwitchParameter_5}]
   properties {name=MetallicOutput, nodeColor=<rgba 0 1 0.7647 1>}
  }
  MaterialExpressionNamedRerouteDeclaration_29 {
   class=MaterialExpressionNamedRerouteDeclaration
   inputs [None {from=MaterialExpressionStaticSwitchParameter_6}]
   properties {name=RoughnessOutput, nodeColor=<rgba 0 1 0.7647 1>}
  }
  MaterialExpressionNamedRerouteDeclaration_3 {
   class=MaterialExpressionNamedRerouteDeclaration
   inputs [None {from=MaterialExpressionReroute_9, output=Roughness}]
   properties {name=Roughness, nodeColor=<rgba 0 0.1434 1 1>}
  }
  MaterialExpressionNamedRerouteDeclaration_30 {
   class=MaterialExpressionNamedRerouteDeclaration
   inputs [None {from=MaterialExpressionMaterialFunctionCall_12, output=TrimSheet_Mask}]
   properties {name=AutoUniMask, nodeColor=<rgba 0.3125 1 0 1>}
  }
  MaterialExpressionNamedRerouteDeclaration_31 {
   class=MaterialExpressionNamedRerouteDeclaration
   inputs [None {from=MaterialExpressionReroute_61, output=SkinMaskClearCoat}]
   properties {name=SkinMaskClearCoatTexture, desc=OPTIMIZATION, nodeColor=<rgba 0.6038 0.3325 0.013 1>}
  }
  MaterialExpressionNamedRerouteDeclaration_32 {
   class=MaterialExpressionNamedRerouteDeclaration
   inputs [None {from=MaterialExpressionComponentMask_5}]
   properties {name=RGBMask, nodeColor=<rgba 0.9453 0 1 1>}
  }
  MaterialExpressionNamedRerouteDeclaration_33 {
   class=MaterialExpressionNamedRerouteDeclaration
   inputs [None {from=MaterialExpressionStaticBoolParameter_0}]
   properties {name=bUseAutoUniLogo, nodeColor=<rgba 0.849 0.0213 0 1>}
  }
  MaterialExpressionNamedRerouteDeclaration_4 {
   class=MaterialExpressionNamedRerouteDeclaration
   inputs [None {from=MaterialExpressionReroute_32, output=ClearCoat Thickness}]
   properties {name=CCThickness, nodeColor=<rgba 0 0.1434 1 1>}
  }
  MaterialExpressionNamedRerouteDeclaration_5 {
   class=MaterialExpressionNamedRerouteDeclaration
   inputs [None {from=MaterialExpressionReroute_33, output=ClearCoat Roughness}]
   properties {name=CCRoughness, nodeColor=<rgba 0 0.1434 1 1>}
  }
  MaterialExpressionNamedRerouteDeclaration_6 {
   class=MaterialExpressionNamedRerouteDeclaration
   inputs [None {from=MaterialExpressionReroute_0, output=DetailBase}]
   properties {name=DetailBaseTexture, nodeColor=<rgba 0.6038 0.3325 0.013 1>}
  }
  MaterialExpressionNamedRerouteDeclaration_7 {
   class=MaterialExpressionNamedRerouteDeclaration
   inputs [None {from=MaterialExpressionReroute_1, output=DetailAccent}]
   properties {name=DetailAccentTexture, nodeColor=<rgba 0.6038 0.3325 0.013 1>}
  }
  MaterialExpressionNamedRerouteDeclaration_8 {
   class=MaterialExpressionNamedRerouteDeclaration
   inputs [None {from=MaterialExpressionReroute_17, output=DetailsMasked}]
   properties {name=DetailsMasked, nodeColor=<rgba 0 1 0.7656 1>}
  }
  MaterialExpressionNamedRerouteDeclaration_9 {
   class=MaterialExpressionNamedRerouteDeclaration
   inputs [None {from=MaterialExpressionMultiply_5}]
   properties {name=UV_Detail, nodeColor=<rgba 1 0.8646 0.8646 1>}
  }
  MaterialExpressionNamedRerouteUsage_0 {class=MaterialExpressionNamedRerouteUsage}
  MaterialExpressionNamedRerouteUsage_1 {class=MaterialExpressionNamedRerouteUsage}
  MaterialExpressionNamedRerouteUsage_10 {class=MaterialExpressionNamedRerouteUsage}
  MaterialExpressionNamedRerouteUsage_11 {class=MaterialExpressionNamedRerouteUsage}
  MaterialExpressionNamedRerouteUsage_12 {class=MaterialExpressionNamedRerouteUsage}
  MaterialExpressionNamedRerouteUsage_13 {class=MaterialExpressionNamedRerouteUsage}
  MaterialExpressionNamedRerouteUsage_14 {class=MaterialExpressionNamedRerouteUsage}
  MaterialExpressionNamedRerouteUsage_15 {class=MaterialExpressionNamedRerouteUsage}
  MaterialExpressionNamedRerouteUsage_16 {class=MaterialExpressionNamedRerouteUsage}
  MaterialExpressionNamedRerouteUsage_17 {class=MaterialExpressionNamedRerouteUsage}
  MaterialExpressionNamedRerouteUsage_18 {class=MaterialExpressionNamedRerouteUsage}
  MaterialExpressionNamedRerouteUsage_19 {class=MaterialExpressionNamedRerouteUsage}
  MaterialExpressionNamedRerouteUsage_2 {class=MaterialExpressionNamedRerouteUsage}
  MaterialExpressionNamedRerouteUsage_20 {class=MaterialExpressionNamedRerouteUsage}
  MaterialExpressionNamedRerouteUsage_21 {class=MaterialExpressionNamedRerouteUsage}
  MaterialExpressionNamedRerouteUsage_22 {class=MaterialExpressionNamedRerouteUsage}
  MaterialExpressionNamedRerouteUsage_23 {class=MaterialExpressionNamedRerouteUsage}
  MaterialExpressionNamedRerouteUsage_24 {class=MaterialExpressionNamedRerouteUsage}
  MaterialExpressionNamedRerouteUsage_25 {class=MaterialExpressionNamedRerouteUsage}
  MaterialExpressionNamedRerouteUsage_26 {class=MaterialExpressionNamedRerouteUsage}
  MaterialExpressionNamedRerouteUsage_27 {class=MaterialExpressionNamedRerouteUsage}
  MaterialExpressionNamedRerouteUsage_28 {class=MaterialExpressionNamedRerouteUsage}
  MaterialExpressionNamedRerouteUsage_29 {class=MaterialExpressionNamedRerouteUsage}
  MaterialExpressionNamedRerouteUsage_3 {class=MaterialExpressionNamedRerouteUsage}
  MaterialExpressionNamedRerouteUsage_30 {class=MaterialExpressionNamedRerouteUsage}
  MaterialExpressionNamedRerouteUsage_31 {class=MaterialExpressionNamedRerouteUsage}
  MaterialExpressionNamedRerouteUsage_32 {class=MaterialExpressionNamedRerouteUsage}
  MaterialExpressionNamedRerouteUsage_33 {class=MaterialExpressionNamedRerouteUsage}
  MaterialExpressionNamedRerouteUsage_34 {class=MaterialExpressionNamedRerouteUsage}
  MaterialExpressionNamedRerouteUsage_35 {class=MaterialExpressionNamedRerouteUsage}
  MaterialExpressionNamedRerouteUsage_36 {class=MaterialExpressionNamedRerouteUsage}
  MaterialExpressionNamedRerouteUsage_37 {class=MaterialExpressionNamedRerouteUsage}
  MaterialExpressionNamedRerouteUsage_38 {class=MaterialExpressionNamedRerouteUsage}
  MaterialExpressionNamedRerouteUsage_39 {class=MaterialExpressionNamedRerouteUsage}
  MaterialExpressionNamedRerouteUsage_4 {class=MaterialExpressionNamedRerouteUsage}
  MaterialExpressionNamedRerouteUsage_42 {class=MaterialExpressionNamedRerouteUsage}
  MaterialExpressionNamedRerouteUsage_43 {class=MaterialExpressionNamedRerouteUsage}
  MaterialExpressionNamedRerouteUsage_45 {class=MaterialExpressionNamedRerouteUsage}
  MaterialExpressionNamedRerouteUsage_5 {class=MaterialExpressionNamedRerouteUsage}
  MaterialExpressionNamedRerouteUsage_6 {class=MaterialExpressionNamedRerouteUsage}
  MaterialExpressionNamedRerouteUsage_7 {class=MaterialExpressionNamedRerouteUsage}
  MaterialExpressionNamedRerouteUsage_8 {class=MaterialExpressionNamedRerouteUsage}
  MaterialExpressionNamedRerouteUsage_9 {class=MaterialExpressionNamedRerouteUsage}
  MaterialExpressionPinBase_0 {
   class=MaterialExpressionPinBase
   inputs []
   properties {reroutePins [SkinMaskTexture {expression=&self:MaterialExpressionReroute_10}]}
  }
  MaterialExpressionPinBase_1 {
   class=MaterialExpressionPinBase
   inputs [
    Metallic {from=MaterialExpressionLinearInterpolate_1, output=""}
    Roughness {from=MaterialExpressionLinearInterpolate_3, output=""}
   ]
   properties {reroutePins [
     Metallic {expression=&self:MaterialExpressionReroute_6}
     Roughness {expression=&self:MaterialExpressionReroute_9}
    ]}
  }
  MaterialExpressionPinBase_10 {
   class=MaterialExpressionPinBase
   inputs []
   properties {reroutePins [
     BaseColor {expression=&self:MaterialExpressionReroute_27}
     BaseColorTexture {expression=&self:MaterialExpressionReroute_29}
     SkinMaskTexture {expression=&self:MaterialExpressionReroute_30}
    ]}
  }
  MaterialExpressionPinBase_11 {
   class=MaterialExpressionPinBase
   inputs [BaseColor {from=MaterialExpressionStaticSwitchParameter_0, output=""}]
   properties {reroutePins [BaseColor {expression=&self:MaterialExpressionReroute_28}]}
  }
  MaterialExpressionPinBase_12 {
   class=MaterialExpressionPinBase
   inputs []
   properties {reroutePins [
     SkinMaskTexture {expression=&self:MaterialExpressionReroute_11}
     BottomNormalBaseTexture {expression=&self:MaterialExpressionReroute_35}
     BottomNormalAccentTexture {expression=&self:MaterialExpressionReroute_36}
    ]}
  }
  MaterialExpressionPinBase_13 {
   class=MaterialExpressionPinBase
   inputs [
    ClearCoat Thickness {from=MaterialExpressionLinearInterpolate_12, output=""}
    ClearCoat Roughness {from=MaterialExpressionLinearInterpolate_18, output=""}
    ClearCoatBottomNormal {from=MaterialExpressionMaterialFunctionCall_7, output=Result}
   ]
   properties {reroutePins [
     ClearCoat Thickness {expression=&self:MaterialExpressionReroute_32}
     ClearCoat Roughness {expression=&self:MaterialExpressionReroute_33}
     ClearCoatBottomNormal {expression=&self:MaterialExpressionReroute_37}
    ]}
  }
  MaterialExpressionPinBase_14 {
   class=MaterialExpressionPinBase
   inputs []
   properties {reroutePins [
     DetailBaseTexture {expression=&self:MaterialExpressionReroute_31}
     DetailAccentTexture {expression=&self:MaterialExpressionReroute_34}
     SkinMaskTexture {expression=&self:MaterialExpressionReroute_16}
    ]}
  }
  MaterialExpressionPinBase_15 {
   class=MaterialExpressionPinBase
   inputs [DetailsMasked {from=MaterialExpressionLinearInterpolate_15, output=""}]
   properties {reroutePins [DetailsMasked {expression=&self:MaterialExpressionReroute_17}]}
  }
  MaterialExpressionPinBase_16 {
   class=MaterialExpressionPinBase
   inputs []
   properties {reroutePins [RGB {expression=&self:MaterialExpressionReroute_42}]}
  }
  MaterialExpressionPinBase_17 {
   class=MaterialExpressionPinBase
   inputs [
    SharpMasksSwitchOutput {from=MaterialExpressionStaticSwitchParameter_1, output=""}
    TextureRawOutput {from=MaterialExpressionStaticSwitchParameter_7, output=""}
   ]
   properties {reroutePins [
     SharpMasksSwitchOutput {expression=&self:MaterialExpressionReroute_45}
     TextureRawOutput {expression=&self:MaterialExpressionReroute_46}
    ]}
  }
  MaterialExpressionPinBase_18 {
   class=MaterialExpressionPinBase
   inputs []
   properties {reroutePins [
     BaseColorInput {expression=&self:MaterialExpressionReroute_47}
     TextureInput {expression=&self:MaterialExpressionReroute_48}
    ]}
  }
  MaterialExpressionPinBase_19 {
   class=MaterialExpressionPinBase
   inputs [Result {from=MaterialExpressionStaticSwitchParameter_2, output=""}]
   properties {reroutePins [Result {expression=&self:MaterialExpressionReroute_49}]}
  }
  MaterialExpressionPinBase_2 {
   class=MaterialExpressionPinBase
   inputs []
   properties {reroutePins [
     UV {expression=&self:MaterialExpressionReroute_4}
     UV Detail {expression=&self:MaterialExpressionReroute_12}
     UV BottomNormal {expression=&self:MaterialExpressionReroute_41}
     LOD {expression=&self:MaterialExpressionReroute_21}
     BottomNormalFlatness {expression=&self:MaterialExpressionReroute_39}
    ]}
  }
  MaterialExpressionPinBase_20 {
   class=MaterialExpressionPinBase
   inputs []
   properties {reroutePins [BaseColor {expression=&self:MaterialExpressionReroute_54}]}
  }
  MaterialExpressionPinBase_21 {
   class=MaterialExpressionPinBase
   inputs [Output {from=MaterialExpressionStaticSwitchParameter_12, output=""}]
   properties {reroutePins [Output {expression=&self:MaterialExpressionReroute_55}]}
  }
  MaterialExpressionPinBase_22 {
   class=MaterialExpressionPinBase
   inputs []
   properties {reroutePins [RGBA {expression=&self:MaterialExpressionReroute_56}]}
  }
  MaterialExpressionPinBase_23 {
   class=MaterialExpressionPinBase
   inputs [
    SkinMask {from=MaterialExpressionStaticSwitchParameter_11, output=""}
    SkinMaskClearCoat {from=MaterialExpressionStaticSwitchParameter_9, output=""}
   ]
   properties {reroutePins [
     SkinMask {expression=&self:MaterialExpressionReroute_59}
     SkinMaskClearCoat {expression=&self:MaterialExpressionReroute_61}
    ]}
  }
  MaterialExpressionPinBase_3 {
   class=MaterialExpressionPinBase
   inputs [
    Curvature {from=MaterialExpressionTextureSampleParameter2D_0, output=R}
    AO {from=MaterialExpressionTextureSampleParameter2D_0, output=R}
    PBR {from=MaterialExpressionAppendVector_0, output=""}
    SkinMask {from=MaterialExpressionTextureSampleParameter2D_1, output=RGB}
    BaseColor {from=MaterialExpressionTextureSampleParameter2D_2, output=RGB}
    DetailBase {from=MaterialExpressionTextureSampleParameter2D_3, output=RGB}
    DetailAccent {from=MaterialExpressionTextureSampleParameter2D_4, output=RGB}
    BodyNormalLOD {from=MaterialExpressionLinearInterpolate_4, output=""}
    BottomNormalBase {from=MaterialExpressionMaterialFunctionCall_1, output=Result}
    BottomNormalAccent {from=MaterialExpressionMaterialFunctionCall_9, output=Result}
   ]
   properties {reroutePins [
     Curvature {expression=&self:MaterialExpressionReroute_22}
     AO {expression=&self:MaterialExpressionReroute_19}
     PBR {expression=&self:MaterialExpressionReroute_53}
     SkinMask {expression=&self:MaterialExpressionReroute_44}
     BaseColor {expression=&self:MaterialExpressionReroute_7}
     DetailBase {expression=&self:MaterialExpressionReroute_0}
     DetailAccent {expression=&self:MaterialExpressionReroute_1}
     BodyNormalLOD {expression=&self:MaterialExpressionReroute_14}
     BottomNormalBase {expression=&self:MaterialExpressionReroute_15}
     BottomNormalAccent {expression=&self:MaterialExpressionReroute_38}
    ]}
  }
  MaterialExpressionPinBase_4 {class=MaterialExpressionPinBase, inputs [], properties {}}
  MaterialExpressionPinBase_5 {
   class=MaterialExpressionPinBase
   inputs [
    TeamColor {from=MaterialExpressionVectorParameter_0, output=RGB}
    AccentColor {from=MaterialExpressionVectorParameter_1, output=RGB}
    PaintColor {from=MaterialExpressionVectorParameter_3, output=RGB}
    TrimColor {from=MaterialExpressionVectorParameter_2, output=RGB}
   ]
   properties {reroutePins [
     TeamColor {expression=&self:MaterialExpressionReroute_23}
     AccentColor {expression=&self:MaterialExpressionReroute_24}
     PaintColor {expression=&self:MaterialExpressionReroute_25}
     TrimColor {expression=&self:MaterialExpressionReroute_43}
    ]}
  }
  MaterialExpressionPinBase_6 {
   class=MaterialExpressionPinBase
   inputs []
   properties {reroutePins [
     Roughness {expression=&self:MaterialExpressionReroute_2}
     SkinMask {expression=&self:MaterialExpressionReroute_3}
    ]}
  }
  MaterialExpressionPinBase_7 {
   class=MaterialExpressionPinBase
   inputs [RoughnessMasked {from=MaterialExpressionLinearInterpolate_6, output=""}]
   properties {reroutePins [RoughnessMasked {expression=&self:MaterialExpressionReroute_5}]}
  }
  MaterialExpressionPinBase_8 {
   class=MaterialExpressionPinBase
   inputs []
   properties {reroutePins [
     TeamColor {expression=&self:MaterialExpressionReroute_8}
     AccentColor {expression=&self:MaterialExpressionReroute_13}
     PaintColor {expression=&self:MaterialExpressionReroute_18}
     SkinMaskTexture {expression=&self:MaterialExpressionReroute_20}
    ]}
  }
  MaterialExpressionPinBase_9 {
   class=MaterialExpressionPinBase
   inputs [BaseColor {from=MaterialExpressionLinearInterpolate_8, output=""}]
   properties {reroutePins [BaseColor {expression=&self:MaterialExpressionReroute_26}]}
  }
  MaterialExpressionPower_0 {
   class=MaterialExpressionPower
   inputs [
    Base {from=MaterialExpressionStaticSwitchParameter_7}
    Exp {from=""}
   ]
   properties {constExponent=0.4545}
  }
  MaterialExpressionPower_1 {
   class=MaterialExpressionPower
   inputs [
    Base {from=MaterialExpressionReroute_48, output=TextureInput}
    Exp {from=""}
   ]
   properties {constExponent=0.4545}
  }
  MaterialExpressionPower_5 {
   class=MaterialExpressionPower
   inputs [
    Base {from=MaterialExpressionReroute_52}
    Exp {from=""}
   ]
   properties {constExponent=8}
  }
  MaterialExpressionPower_7 {
   class=MaterialExpressionPower
   inputs [
    Base {from=MaterialExpressionAdd_4}
    Exp {from=""}
   ]
   properties {constExponent=0.25}
  }
  MaterialExpressionReroute_10 {class=MaterialExpressionReroute, inputs [None {from=MaterialExpressionNamedRerouteUsage_12}]}
  MaterialExpressionReroute_11 {class=MaterialExpressionReroute, inputs [None {from=MaterialExpressionNamedRerouteUsage_43}]}
  MaterialExpressionReroute_12 {class=MaterialExpressionReroute, inputs [None {from=MaterialExpressionNamedRerouteUsage_26}]}
  MaterialExpressionReroute_13 {class=MaterialExpressionReroute, inputs [None {from=MaterialExpressionNamedRerouteUsage_9}]}
  MaterialExpressionReroute_16 {class=MaterialExpressionReroute, inputs [None {from=MaterialExpressionNamedRerouteUsage_18}]}
  MaterialExpressionReroute_17 {class=MaterialExpressionReroute, inputs [None {from=MaterialExpressionLinearInterpolate_15}]}
  MaterialExpressionReroute_18 {class=MaterialExpressionReroute, inputs [None {from=MaterialExpressionNamedRerouteUsage_10}]}
  MaterialExpressionReroute_2 {class=MaterialExpressionReroute, inputs [None {from=MaterialExpressionNamedRerouteUsage_5}]}
  MaterialExpressionReroute_20 {class=MaterialExpressionReroute, inputs [None {from=MaterialExpressionNamedRerouteUsage_7}]}
  MaterialExpressionReroute_21 {class=MaterialExpressionReroute, inputs [None {from=MaterialExpressionScalarParameter_1}]}
  MaterialExpressionReroute_23 {class=MaterialExpressionReroute, inputs [None {from=MaterialExpressionVectorParameter_0, output=RGB}]}
  MaterialExpressionReroute_24 {class=MaterialExpressionReroute, inputs [None {from=MaterialExpressionVectorParameter_1, output=RGB}]}
  MaterialExpressionReroute_25 {class=MaterialExpressionReroute, inputs [None {from=MaterialExpressionVectorParameter_3, output=RGB}]}
  MaterialExpressionReroute_26 {class=MaterialExpressionReroute, inputs [None {from=MaterialExpressionLinearInterpolate_8}]}
  MaterialExpressionReroute_27 {class=MaterialExpressionReroute, inputs [None {from=MaterialExpressionReroute_26, output=BaseColor}]}
  MaterialExpressionReroute_28 {class=MaterialExpressionReroute, inputs [None {from=MaterialExpressionStaticSwitchParameter_0}]}
  MaterialExpressionReroute_29 {class=MaterialExpressionReroute, inputs [None {from=MaterialExpressionNamedRerouteUsage_15}]}
  MaterialExpressionReroute_3 {class=MaterialExpressionReroute, inputs [None {from=MaterialExpressionNamedRerouteUsage_4}]}
  MaterialExpressionReroute_30 {class=MaterialExpressionReroute, inputs [None {from=MaterialExpressionNamedRerouteUsage_14}]}
  MaterialExpressionReroute_31 {class=MaterialExpressionReroute, inputs [None {from=MaterialExpressionNamedRerouteUsage_29}]}
  MaterialExpressionReroute_32 {class=MaterialExpressionReroute, inputs [None {from=MaterialExpressionLinearInterpolate_12}]}
  MaterialExpressionReroute_33 {class=MaterialExpressionReroute, inputs [None {from=MaterialExpressionLinearInterpolate_18}]}
  MaterialExpressionReroute_34 {class=MaterialExpressionReroute, inputs [None {from=MaterialExpressionNamedRerouteUsage_30}]}
  MaterialExpressionReroute_35 {class=MaterialExpressionReroute, inputs [None {from=MaterialExpressionNamedRerouteUsage_11}]}
  MaterialExpressionReroute_36 {class=MaterialExpressionReroute, inputs [None {from=MaterialExpressionNamedRerouteUsage_19}]}
  MaterialExpressionReroute_37 {class=MaterialExpressionReroute, inputs [None {from=MaterialExpressionMaterialFunctionCall_7, output=Result}]}
  MaterialExpressionReroute_38 {class=MaterialExpressionReroute, inputs [None {from=MaterialExpressionMaterialFunctionCall_9, output=Result}]}
  MaterialExpressionReroute_39 {class=MaterialExpressionReroute, inputs [None {from=MaterialExpressionScalarParameter_14}]}
  MaterialExpressionReroute_4 {class=MaterialExpressionReroute, inputs [None {from=MaterialExpressionNamedRerouteUsage_25}]}
  MaterialExpressionReroute_40 {class=MaterialExpressionReroute, inputs [None {from=MaterialExpressionReroute_39, output=BottomNormalFlatness}]}
  MaterialExpressionReroute_41 {class=MaterialExpressionReroute, inputs [None {from=MaterialExpressionNamedRerouteUsage_21}]}
  MaterialExpressionReroute_42 {class=MaterialExpressionReroute, inputs [None {from=MaterialExpressionReroute_44, output=SkinMask}]}
  MaterialExpressionReroute_43 {class=MaterialExpressionReroute, inputs [None {from=MaterialExpressionVectorParameter_2, output=RGB}]}
  MaterialExpressionReroute_45 {class=MaterialExpressionReroute, inputs [None {from=MaterialExpressionStaticSwitchParameter_1}]}
  MaterialExpressionReroute_46 {class=MaterialExpressionReroute, inputs [None {from=MaterialExpressionStaticSwitchParameter_7}]}
  MaterialExpressionReroute_47 {class=MaterialExpressionReroute, inputs [None {from=MaterialExpressionReroute_55, output=Output}]}
  MaterialExpressionReroute_48 {class=MaterialExpressionReroute, inputs [None {from=MaterialExpressionNamedRerouteUsage_27}]}
  MaterialExpressionReroute_49 {class=MaterialExpressionReroute, inputs [None {from=MaterialExpressionStaticSwitchParameter_2}]}
  MaterialExpressionReroute_5 {class=MaterialExpressionReroute, inputs [None {from=MaterialExpressionLinearInterpolate_6}]}
  MaterialExpressionReroute_50 {class=MaterialExpressionReroute, inputs [None {from=MaterialExpressionReroute_47, output=BaseColorInput}]}
  MaterialExpressionReroute_51 {class=MaterialExpressionReroute, inputs [None {from=MaterialExpressionReroute_50}]}
  MaterialExpressionReroute_52 {class=MaterialExpressionReroute, inputs [None {from=MaterialExpressionComponentMask_1}]}
  MaterialExpressionReroute_53 {class=MaterialExpressionReroute, inputs [None {from=MaterialExpressionAppendVector_0}]}
  MaterialExpressionReroute_54 {class=MaterialExpressionReroute, inputs [None {from=MaterialExpressionReroute_28, output=BaseColor}]}
  MaterialExpressionReroute_55 {class=MaterialExpressionReroute, inputs [None {from=MaterialExpressionStaticSwitchParameter_12}]}
  MaterialExpressionReroute_56 {class=MaterialExpressionReroute, inputs [None {from=MaterialExpressionReroute_45, output=SharpMasksSwitchOutput}]}
  MaterialExpressionReroute_57 {class=MaterialExpressionReroute, inputs [None {from=MaterialExpressionReroute_54, output=BaseColor}]}
  MaterialExpressionReroute_58 {class=MaterialExpressionReroute, inputs [None {from=MaterialExpressionReroute_54, output=BaseColor}]}
  MaterialExpressionReroute_59 {class=MaterialExpressionReroute, inputs [None {from=MaterialExpressionStaticSwitchParameter_11}]}
  MaterialExpressionReroute_6 {class=MaterialExpressionReroute, inputs [None {from=MaterialExpressionLinearInterpolate_1}]}
  MaterialExpressionReroute_60 {class=MaterialExpressionReroute, inputs [None {from=MaterialExpressionTextureCoordinate_4}]}
  MaterialExpressionReroute_61 {class=MaterialExpressionReroute, inputs [None {from=MaterialExpressionStaticSwitchParameter_9}]}
  MaterialExpressionReroute_62 {class=MaterialExpressionReroute, inputs [None {from=MaterialExpressionReroute_60}]}
  MaterialExpressionReroute_64 {class=MaterialExpressionReroute, inputs [None {from=MaterialExpressionReroute_71}]}
  MaterialExpressionReroute_65 {class=MaterialExpressionReroute, inputs [None {from=MaterialExpressionReroute_69}]}
  MaterialExpressionReroute_69 {class=MaterialExpressionReroute, inputs [None {from=MaterialExpressionNamedRerouteDeclaration_32}]}
  MaterialExpressionReroute_71 {class=MaterialExpressionReroute, inputs [None {from=MaterialExpressionNamedRerouteUsage_42}]}
  MaterialExpressionReroute_8 {class=MaterialExpressionReroute, inputs [None {from=MaterialExpressionNamedRerouteUsage_8}]}
  MaterialExpressionReroute_9 {class=MaterialExpressionReroute, inputs [None {from=MaterialExpressionLinearInterpolate_3}]}
  MaterialExpressionSaturate_0 {class=MaterialExpressionSaturate, inputs [None {from=MaterialExpressionConstantBiasScale_0}]}
  MaterialExpressionSaturate_1 {class=MaterialExpressionSaturate, inputs [None {from=MaterialExpressionMaterialFunctionCall_10, output=G}]}
  MaterialExpressionSaturate_10 {class=MaterialExpressionSaturate, inputs [None {from=MaterialExpressionLinearInterpolate_20}]}
  MaterialExpressionSaturate_2 {class=MaterialExpressionSaturate, inputs [None {from=MaterialExpressionClamp_1}]}
  MaterialExpressionSaturate_3 {class=MaterialExpressionSaturate, inputs [None {from=MaterialExpressionClamp_0}]}
  MaterialExpressionSaturate_4 {class=MaterialExpressionSaturate, inputs [None {from=MaterialExpressionAdd_2}]}
  MaterialExpressionSaturate_5 {class=MaterialExpressionSaturate, inputs [None {from=MaterialExpressionAdd_3}]}
  MaterialExpressionSaturate_6 {class=MaterialExpressionSaturate, inputs [None {from=MaterialExpressionClamp_2}]}
  MaterialExpressionScalarParameter_0 {class=MaterialExpressionScalarParameter, properties {defaultValue=3, group=Finish, parameterName=DetailTiling}}
  MaterialExpressionScalarParameter_1 {class=MaterialExpressionScalarParameter, properties {parameterName=LOD}}
  MaterialExpressionScalarParameter_10 {class=MaterialExpressionScalarParameter, properties {defaultValue=1, group=Finish, parameterName=CCThickness_PaintColor}}
  MaterialExpressionScalarParameter_11 {class=MaterialExpressionScalarParameter, properties {defaultValue=0.1, group=Finish, parameterName=CCRoughness_TeamColor}}
  MaterialExpressionScalarParameter_12 {class=MaterialExpressionScalarParameter, properties {defaultValue=0.4, group=Finish, parameterName=CCRoughness_AccentColor}}
  MaterialExpressionScalarParameter_13 {class=MaterialExpressionScalarParameter, properties {defaultValue=0.4, group=Finish, parameterName=CCRoughness_PaintColor}}
  MaterialExpressionScalarParameter_14 {class=MaterialExpressionScalarParameter, properties {defaultValue=0.85, group=DetailNormal, parameterName=BottomNormalFlatness}}
  MaterialExpressionScalarParameter_15 {class=MaterialExpressionScalarParameter, properties {defaultValue=13, group=Finish, parameterName=BottomNormalTiling}}
  MaterialExpressionScalarParameter_16 {
   class=MaterialExpressionScalarParameter
   properties {desc="0 = TeamColor; 1 = AccentColor; 2 = PaintColor", group=UniversalLogos, parameterName=ColorParamSelector, sliderMax=2, sortPriority=9}
  }
  MaterialExpressionScalarParameter_17 {
   class=MaterialExpressionScalarParameter
   properties {
    desc="0 = Team Finish; 1 = Accent Finish; 2 = Paint Finish"
    group=UniversalLogos
    parameterName=FinishParamSelector
    sliderMax=2
    sortPriority=7
   }
  }
  MaterialExpressionScalarParameter_2 {class=MaterialExpressionScalarParameter, properties {defaultValue=0.7, group=Finish, parameterName=Metallic_TeamColor}}
  MaterialExpressionScalarParameter_3 {class=MaterialExpressionScalarParameter, properties {defaultValue=0.7, group=Finish, parameterName=Metallic_AccentColor}}
  MaterialExpressionScalarParameter_4 {class=MaterialExpressionScalarParameter, properties {defaultValue=0.7, group=Finish, parameterName=Metallic_PaintColor}}
  MaterialExpressionScalarParameter_5 {class=MaterialExpressionScalarParameter, properties {defaultValue=0.4, group=Finish, parameterName=Rough_TeamColor}}
  MaterialExpressionScalarParameter_6 {class=MaterialExpressionScalarParameter, properties {defaultValue=0.5, group=Finish, parameterName=Rough_AccentColor}}
  MaterialExpressionScalarParameter_7 {class=MaterialExpressionScalarParameter, properties {defaultValue=0.5, group=Finish, parameterName=Rough_PaintColor}}
  MaterialExpressionScalarParameter_8 {class=MaterialExpressionScalarParameter, properties {defaultValue=1, group=Finish, parameterName=CCThickness_TeamColor}}
  MaterialExpressionScalarParameter_9 {class=MaterialExpressionScalarParameter, properties {defaultValue=1, group=Finish, parameterName=CCThickness_AccentColor}}
  MaterialExpressionSetMaterialAttributes_0 {
   class=MaterialExpressionSetMaterialAttributes
   inputs [
    MaterialAttributes {from="", output=""}
    BaseColor {from=MaterialExpressionNamedRerouteUsage_16, output=""}
    Metallic {from=MaterialExpressionNamedRerouteUsage_36, output=""}
    Roughness {from=MaterialExpressionNamedRerouteUsage_37, output=""}
    CustomData0 {from=MaterialExpressionNamedRerouteUsage_22, output=""}
    CustomData1 {from=MaterialExpressionNamedRerouteUsage_23, output=""}
    Normal {from=MaterialExpressionNamedRerouteUsage_3, output=""}
    ClearCoatBottomNormal {from=MaterialExpressionNamedRerouteUsage_32, output=""}
   ]
   properties {attributeSetTypes [
     #1
     #2
     #3
     #4
     #5
     #6
     #7
    ]}
  }
  MaterialExpressionStaticBoolParameter_0 {
   class=MaterialExpressionStaticBoolParameter
   inputs []
   properties {defaultValue=true, desc=Exposes the Auto Universal logo/trimsheet funcionality}
  }
  MaterialExpressionStaticSwitchParameter_0 {
   class=MaterialExpressionStaticSwitchParameter
   inputs [
    True {from=MaterialExpressionLinearInterpolate_10, output=""}
    False {from=MaterialExpressionReroute_27, output=BaseColor}
   ]
   properties {parameterName=UseBaseColorTexture}
  }
  MaterialExpressionStaticSwitchParameter_1 {
   class=MaterialExpressionStaticSwitchParameter
   inputs [
    True {from=MaterialExpressionSaturate_0, output=""}
    False {from=MaterialExpressionStaticSwitchParameter_7, output=""}
   ]
   properties {
    desc="DO NOT USE FOR NEW ASSETS! This is intended for porting content from Rocket League only."
    group=Legacy_PORTINGONLY
    parameterName=SharpMasks
   }
  }
  MaterialExpressionStaticSwitchParameter_10 {
   class=MaterialExpressionStaticSwitchParameter
   inputs [
    True {from=MaterialExpressionLinearInterpolate_24, output=""}
    False {from=MaterialExpressionNamedRerouteDeclaration_32, output=""}
   ]
   properties {
    desc="OPTIMIZATION - USE ONLY IF NECESSARY! Adds control of the Auto Universal logo into the Clear Coat via its associated finish group (Team/Accent/Paint)."
    group=UniversalLogos
    parameterName=EnableLogoClearCoat
    sortPriority=5
   }
  }
  MaterialExpressionStaticSwitchParameter_11 {
   class=MaterialExpressionStaticSwitchParameter
   inputs [
    True {from=MaterialExpressionStaticSwitchParameter_14, output=""}
    False {from=MaterialExpressionReroute_64, output=""}
   ]
   properties {defaultValue=true, desc=Exposes the Auto Universal logo/trimsheet funcionality, parameterName=UseAutoUniversalLogo}
  }
  MaterialExpressionStaticSwitchParameter_12 {
   class=MaterialExpressionStaticSwitchParameter
   inputs [
    True {from=MaterialExpressionLinearInterpolate_9, output=""}
    False {from=MaterialExpressionReroute_57, output=""}
   ]
   properties {defaultValue=true, desc=Exposes the Auto Universal logo/trimsheet funcionality, parameterName=UseAutoUniversalLogo}
  }
  MaterialExpressionStaticSwitchParameter_14 {
   class=MaterialExpressionStaticSwitchParameter
   inputs [
    True {from=MaterialExpressionLinearInterpolate_24, output=""}
    False {from=MaterialExpressionReroute_71, output=""}
   ]
   properties {
    desc="Exposes controls for associating the PBR of the Auto Universal logo with an existing finish group (Team/Accent/Paint)"
    group=UniversalLogos
    parameterName=EnableLogoFinishSelector
    sortPriority=4
   }
  }
  MaterialExpressionStaticSwitchParameter_15 {
   class=MaterialExpressionStaticSwitchParameter
   inputs [
    True {from=MaterialExpressionStaticSwitchParameter_10, output=""}
    False {from=MaterialExpressionReroute_69, output=""}
   ]
   properties {
    desc="Exposes controls for associating the PBR of the Auto Universal logo with an existing finish group (Team/Accent/Paint)"
    group=UniversalLogos
    parameterName=EnableLogoFinishSelector
    sortPriority=4
   }
  }
  MaterialExpressionStaticSwitchParameter_2 {
   class=MaterialExpressionStaticSwitchParameter
   inputs [
    True {from=MaterialExpressionMultiply_1, output=""}
    False {from=MaterialExpressionReroute_51, output=""}
   ]
   properties {
    desc="DO NOT USE FOR NEW ASSETS! This is intended for porting content from Rocket League only."
    group=Legacy_PORTINGONLY
    parameterName=SharpMasks
   }
  }
  MaterialExpressionStaticSwitchParameter_3 {
   class=MaterialExpressionStaticSwitchParameter
   inputs [
    True {from=MaterialExpressionTextureCoordinate_4, output=""}
    False {from=MaterialExpressionTextureCoordinate_5, output=""}
   ]
   properties {group=DetailNormal, parameterName=UVChannel2}
  }
  MaterialExpressionStaticSwitchParameter_4 {
   class=MaterialExpressionStaticSwitchParameter
   inputs [
    True {from=MaterialExpressionReroute_62, output=""}
    False {from=MaterialExpressionStaticSwitchParameter_3, output=""}
   ]
   properties {defaultValue=true, desc=Exposes the Auto Universal logo/trimsheet funcionality, parameterName=UseAutoUniversalLogo}
  }
  MaterialExpressionStaticSwitchParameter_5 {
   class=MaterialExpressionStaticSwitchParameter
   inputs [
    True {from=MaterialExpressionComponentMask_2, output=""}
    False {from=MaterialExpressionNamedRerouteUsage_28, output=""}
   ]
   properties {
    desc="When using, ensure PBR texture has \"Compress Without Alpha\" set to False. "
    group=Finish
    parameterName=UsePBR_BodyTextures
    sortPriority=-10
   }
  }
  MaterialExpressionStaticSwitchParameter_6 {
   class=MaterialExpressionStaticSwitchParameter
   inputs [
    True {from=MaterialExpressionComponentMask_3, output=""}
    False {from=MaterialExpressionNamedRerouteDeclaration_11, output=""}
   ]
   properties {
    desc="When using, ensure PBR texture has \"Compress Without Alpha\" set to False. "
    group=Finish
    parameterName=UsePBR_BodyTextures
    sortPriority=-10
   }
  }
  MaterialExpressionStaticSwitchParameter_7 {
   class=MaterialExpressionStaticSwitchParameter
   inputs [
    True {from=MaterialExpressionAppendVector_2, output=""}
    False {from=MaterialExpressionAppendVector_3, output=""}
   ]
   properties {defaultValue=true, desc=Exposes the Auto Universal logo/trimsheet funcionality, parameterName=UseAutoUniversalLogo}
  }
  MaterialExpressionStaticSwitchParameter_8 {
   class=MaterialExpressionStaticSwitchParameter
   inputs [
    True {from=MaterialExpressionNamedRerouteUsage_6, output=""}
    False {from=MaterialExpressionLinearInterpolate_16, output=""}
   ]
   properties {defaultValue=true, group=UniversalLogos, parameterName=UseDiffuseAsColor, sortPriority=2}
  }
  MaterialExpressionStaticSwitchParameter_9 {
   class=MaterialExpressionStaticSwitchParameter
   inputs [
    True {from=MaterialExpressionStaticSwitchParameter_15, output=""}
    False {from=MaterialExpressionReroute_65, output=""}
   ]
   properties {defaultValue=true, desc=Exposes the Auto Universal logo/trimsheet funcionality, parameterName=UseAutoUniversalLogo}
  }
  MaterialExpressionTextureCoordinate_0 {class=MaterialExpressionTextureCoordinate}
  MaterialExpressionTextureCoordinate_1 {class=MaterialExpressionTextureCoordinate, properties {coordinateIndex=1}}
  MaterialExpressionTextureCoordinate_4 {class=MaterialExpressionTextureCoordinate, properties {coordinateIndex=1}}
  MaterialExpressionTextureCoordinate_5 {class=MaterialExpressionTextureCoordinate}
  MaterialExpressionTextureSampleParameter2D_0 {
   class=MaterialExpressionTextureSampleParameter2D
   inputs [
    UVs {from="", output=""}
    Apply View MipBias {from="", output=""}
   ]
   properties {
    group=Texture
    parameterName=BodyTextures
    samplerSource=SSM_Wrap_WorldGroupSettings
    samplerType=SAMPLERTYPE_LinearColor
    texture=/VehicleCosmetics/Textures/T_Black_Linear
   }
  }
  MaterialExpressionTextureSampleParameter2D_1 {
   class=MaterialExpressionTextureSampleParameter2D
   inputs [
    UVs {from=MaterialExpressionReroute_4, output=UV}
    Apply View MipBias {from="", output=""}
   ]
   properties {group=Texture, parameterName=SkinMask, samplerSource=SSM_Wrap_WorldGroupSettings, texture=/VehicleCosmetics/Textures/T_Black}
  }
  MaterialExpressionTextureSampleParameter2D_2 {
   class=MaterialExpressionTextureSampleParameter2D
   inputs [
    UVs {from=MaterialExpressionReroute_4, output=UV}
    Apply View MipBias {from="", output=""}
   ]
   properties {
    constCoordinate=1
    group=Texture
    parameterName=SkinBaseColor
    samplerSource=SSM_Wrap_WorldGroupSettings
    texture=&T_Black
   }
  }
  MaterialExpressionTextureSampleParameter2D_3 {
   class=MaterialExpressionTextureSampleParameter2D
   inputs [
    UVs {from=MaterialExpressionReroute_12, output=UV Detail}
    Apply View MipBias {from="", output=""}
   ]
   properties {group=Finish, parameterName=DetailBase, samplerSource=SSM_Wrap_WorldGroupSettings, texture=/VehicleCosmetics/Textures/T_White}
  }
  MaterialExpressionTextureSampleParameter2D_4 {
   class=MaterialExpressionTextureSampleParameter2D
   inputs [
    UVs {from=MaterialExpressionReroute_12, output=UV Detail}
    Apply View MipBias {from="", output=""}
   ]
   properties {group=Finish, parameterName=DetailAccent, samplerSource=SSM_Wrap_WorldGroupSettings, texture=&T_White}
  }
  MaterialExpressionTextureSampleParameter2D_5 {
   class=MaterialExpressionTextureSampleParameter2D
   inputs [
    UVs {from="", output=""}
    Apply View MipBias {from="", output=""}
   ]
   properties {group=Texture, parameterName=BodyNormalLOD, samplerType=SAMPLERTYPE_Normal, texture=/VehicleCosmetics/Textures/T_BlankNormal}
  }
  MaterialExpressionTextureSampleParameter2D_6 {
   class=MaterialExpressionTextureSampleParameter2D
   inputs [
    UVs {from=MaterialExpressionReroute_41, output=UV BottomNormal}
    Apply View MipBias {from="", output=""}
   ]
   properties {
    group=Finish
    parameterName=BottomNormalBase
    samplerType=SAMPLERTYPE_Normal
    texture=/VehicleCosmetics/SharedTextures/T_Wrap_metallicFlakes_Small
   }
  }
  MaterialExpressionTextureSampleParameter2D_7 {
   class=MaterialExpressionTextureSampleParameter2D
   inputs [
    UVs {from=MaterialExpressionReroute_41, output=UV BottomNormal}
    Apply View MipBias {from="", output=""}
   ]
   properties {group=Finish, parameterName=BottomNormalAccent, samplerType=SAMPLERTYPE_Normal, texture=/VehicleCosmetics/Textures/T_BlankNormal_B}
  }
  MaterialExpressionVectorParameter_0 {class=MaterialExpressionVectorParameter, properties {group=Color, parameterName=BaseColor, sortPriority=1, defaultValue=<rgba 0 0.2 1 1>}}
  MaterialExpressionVectorParameter_1 {class=MaterialExpressionVectorParameter, properties {group=Color, parameterName=AccentColor, sortPriority=2, defaultValue=<rgba 1 1 1 1>}}
  MaterialExpressionVectorParameter_2 {
   class=MaterialExpressionVectorParameter
   properties {group=Color, parameterName=TrimColor, sortPriority=4, defaultValue=<rgba 0.03 0.03 0.03 0>}
  }
  MaterialExpressionVectorParameter_3 {class=MaterialExpressionVectorParameter, properties {group=Color, parameterName=PaintColor, sortPriority=3, defaultValue=<rgba 0 1 0 1>}}
 }}
